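# Niagara effect for the ModernMix emote that flashes short sprite bursts on the character's hands, with a wrist airglow.
# HandBurst spawns two spawn groups, each following its own skeletal mesh socket, with curve-driven color and size and distance culling.
# WristAirglow bursts in spawn group 2, which switches socket, size and lifetime, and scales alpha by front end, time of day and platform.
# Both emitters use two sprite renderers and write separate flare color and size attributes.
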NiagaraSystem /BRCosmetics/Animation/Game/MainPlayer/Emotes/ModernMix/FX/NS_Emote_ModernMix_HandsBurst
dependencies {
 usedDataInterfaces [
  &NiagaraDataInterfaceColorCurve
  &NiagaraDataInterfaceCurve
  &NiagaraDataInterfaceArrayFloat3
  &NiagaraDataInterfacePlatformSet
  &NiagaraDataInterfaceSkeletalMesh
 ]
 usedDynamicInputs []
 usedModules [
  &self:SystemSpawnScript.SystemScriptSource.SystemScriptGraph.NiagaraNodeAssignment_1.SetVariables_#1
  &SystemState
 ]
}
emitters {
 HandBurst {
  emitterSpawnScript [/SetVariables(Emitter.GlowDepthFade=25) {isSetParameters=true}]
  emitterUpdateScript [
   &EmitterState(
    Enable Distance Culling=true
    Life Cycle Mode=::NewEnumerator1
    Loop Behavior=::NewEnumerator1
    Loop Duration=1
    Max Distance Response=::KillImmediately
    MaxDistance=3500
    Module.bUseMaxDistance=true
    Module.bUseMinDistance=false
    Scalability Mode=::NewEnumerator1
    UseLoopDelay=false
   )
   &SpawnBurst_Instantaneous(Spawn Count=2, Use Spawn Probability=false)
  ]
  particleSpawnScript [
   &FN_SetSpawnGroup(Number of Spawn Groups=2)
   &InitializeParticle(
    Color=<rgba 1 0.7044 0.4792 2>
    Lifetime=0.4
    Position Offset Coordinate Space=::Local
    Sprite Rotation Mode=::NewEnumerator2
    Sprite Size Mode=::NewEnumerator3
    Uniform Sprite Size=130
    UsePositionOffset=true
    Write Material Random=true
   )
   &CameraOffset(Camera Offset Amount=-6)
  ]
  particleUpdateScript [
   &ParticleState()
   &ScaleColor(
    Scale Alpha=&FloatFromCurve(FloatCurve=DataInterface() {propertyValues {Curve {keys [<at 0 0.5>,<at 1 0>]}}}, Scale Curve=0.3)
    Scale RGB=&MakeVectorFromLinearColorRGB(Color=&ColorFromCurve(ColorCurve=DataInterface() {propertyValues {
        AlphaCurve {keys [<at 0 1 RCIM_Cubic>,<at 1 0 RCIM_Cubic>]}
        BlueCurve {keys [<at 0.0117 0>, {arriveTangent=0.24, interpMode=RCIM_Cubic, leaveTangent=0.24, time=0.0974, value=0}, <at 1 0.2372 RCIM_Cubic>]}
        GreenCurve {keys [<at 0.0117 1>, {arriveTangent=-0.4718, interpMode=RCIM_Cubic, leaveTangent=-0.4718, time=0.0974, value=5}, <at 1 0.5337 RCIM_Cubic>]}
        RedCurve {keys [<at 0.0117 0.9104>, {arriveTangent=0.0907, interpMode=RCIM_Cubic, leaveTangent=0.0907, time=0.0974, value=3.5535}, <at 1 1 RCIM_Cubic>]}
       }}))
    ScaleA=true
    ScaleRGB=true
   )
   &ScaleSpriteSize(Scale Factor=&Vector2DFromFloat(Value=&FloatFromCurve(
      FloatCurve=DataInterface() {propertyValues {Curve {keys [{arriveTangent=2.7513, interpMode=RCIM_Cubic, leaveTangent=2.7513, tangentMode=RCTM_User, time=0, value=0}, <at 1 1.2>]}}}
      Scale Curve=0.5
     )))
   /SetVariables(
    Particles.FlareColor=&MultiplyLinearColor(Linear Color=$Particles.Initial.Color, Scale Factor=&MakeVector4_FromFloat(Float=0.85))
    Particles.FlareSize:Vector2f=&Multiply_Vector2D_ByFloat(
     Float=&FloatFromCurve(FloatCurve=DataInterface() {propertyValues {Curve {keys [
          {interpMode=RCIM_Cubic, tangentMode=RCTM_User, time=0, value=0}
          {interpMode=RCIM_Cubic, tangentMode=RCTM_User, time=0.3, value=0.95}
          {tangentMode=RCTM_User, time=1, value=0}
         ]}}})
     Vector2D=$Particles.Initial.SpriteSize
    )
    Particles.FlashCameraOffset=-7
   ) {isSetParameters=true}
   &SkeletalMeshLocation(
    Bone / Socket Index=$Particles.SpawnGroup
    MaskSpawnGroup=false
    Mesh Sampling Type=::NewEnumerator5
    Sampled Position Offset=&SelectVectorFromArray(Direct Array Index=$Particles.SpawnGroup, Vector Selection Array=DataInterface() {propertyValues {FloatData [<0 0 0>, <0 0 0>]}})
    Skeletal Mesh=$System.Skeletal Mesh
    Socket Sampling Mode=::NewEnumerator5
   )
  ]
  rendererClasses [&NiagaraSpriteRendererProperties]
  renderers [
   {rendererClass=&NiagaraSpriteRendererProperties}
   {rendererClass=&NiagaraSpriteRendererProperties, rendererIndex=1}
  ]
 }
 WristAirglow {
  emitterUpdateScript [
   &EmitterState(Inactive Response=::NewEnumerator1, Life Cycle Mode=::NewEnumerator1, Loop Behavior=::NewEnumerator1, Loop Duration=1, UseLoopDelay=false)
   &SpawnBurst_Instantaneous(Spawn Group=2, Use Spawn Probability=false, UseLoopCountLimit=false) {module=SpawnBurst_Instantaneous001}
  ]
  particleSpawnScript [
   &InitializeParticle(
    Color=<rgba 1 0.7036 0.4348 0.5>
    Lifetime=&MergeHelper_FloatSwitch(0=10.9, 1=10.9, 2=0.315, 3=0.3, Index=&SpawnGroup(), Num=::NewEnumerator2)
    Sprite Size Mode=::NewEnumerator3
    Uniform Sprite Size=&MergeHelper_FloatSwitch(0=120, 1=120, 2=320, 3=315, Index=&SpawnGroup(), Num=::NewEnumerator2)
    UsePositionOffset=false
    Write Material Random=true
   ) {module=InitializeParticle001}
   &SpriteFacingAndAlignment(Alignment Coordinate Space=::Local, Sprite Alignment=<1 0.25 0>, WriteAlignment=true, WriteFacing=false)
   &CameraOffset(Camera Offset Amount=&MergeHelper_FloatSwitch(0=-2, 1=-2, 2=8, 3=8, Index=&SpawnGroup(), Num=::NewEnumerator2))
   /SetVariables(Particles.FlareVisibility:NiagaraInt32=&SpawnGroup()) {isSetParameters=true}
  ]
  particleUpdateScript [
   &SkeletalMeshLocation(
    Bone / Socket Index=&MergeHelper_IntSwitch(0=0, 1=1, 2=4, 3=4, 4=?, 5=?, Index=&SpawnGroup(), Num=::NewEnumerator2)
    MaskSpawnGroup=false
    Mesh Sampling Type=::NewEnumerator5
    Sampled Position Offset Coordinate Space:ENiagaraSkelMeshTransforms=::NewEnumerator2
    Skeletal Mesh=$System.Skeletal Mesh
    Socket Sampling Mode=::NewEnumerator5
   )
   &ParticleState()
   &ScaleSpriteSize(
    Scale Sprite Size Mode=::NewEnumerator0
    Uniform Scale Factor=&MergeHelper_FloatSwitch(
     0=1
     1=1
     2=&FloatFromCurve(FloatCurve=DataInterface() {propertyValues {Curve {keys [{arriveTangent=2.3562, interpMode=RCIM_Cubic, leaveTangent=2.3562, tangentMode=RCTM_User, time=0, value=0}, <at 1 1>]}}})
     3=&FloatFromCurve(FloatCurve=DataInterface() {propertyValues {Curve {keys [{arriveTangent=2.3562, interpMode=RCIM_Cubic, leaveTangent=2.3562, tangentMode=RCTM_User, time=0, value=0}, <at 1 1>]}}})
     Index=&SpawnGroup()
     Num=::NewEnumerator2
    )
   )
   &ScaleColor(
    Scale Alpha=&Multiply_Float(
     A=&FN_FrontEndFloatModifier(Gameplay=1, Front End=0.78)
     B=&Multiply_Float(
      A=&FNTimeOfDayScalar(Night Scale=0.7, Morning Scale=0.9, Evening Scale=0.8)
      B=&FloatFromCurve(
       FloatCurve=DataInterface() {propertyValues {Curve {keys [
           {arriveTangent=-1.8326, interpMode=RCIM_Cubic, leaveTangent=-1.8326, tangentMode=RCTM_User, time=0, value=1}
           {arriveTangent=-0.5891, tangentMode=RCTM_User, time=1, value=0}
          ]}}}
       Scale Curve=&FN_PickFloatDependOnPlatform(TargetPlatform=DataInterface() {propertyValues {Platforms {qualityLevelMask=-29}}}, TargetPlatformActive=0.28, TargetPlatformNoActive=0.4)
      )
     )
    )
    ScaleA=true
    ScaleRGB=true
   ) {module=ScaleColor002}
   /SetVariables(
    Particles.FlareColor=&MakeLinearColorFromVectorAndFloat(
     "Vector (RGB)"=&MakeVectorFromLinearColorRGB(Color=<rgba 2 1.2112 0.7396 1>)
     "Float (Alpha)"=&FloatFromCurve(
      FloatCurve=DataInterface() {propertyValues {Curve {keys [
          {arriveTangent=-1.8326, interpMode=RCIM_Cubic, leaveTangent=-1.8326, tangentMode=RCTM_User, time=0, value=1}
          {arriveTangent=-0.5891, tangentMode=RCTM_User, time=1, value=0}
         ]}}}
      Scale Curve=&FN_PickFloatDependOnPlatform(TargetPlatform=DataInterface() {propertyValues {Platforms {qualityLevelMask=-29}}}, TargetPlatformActive=0.28, TargetPlatformNoActive=0.4)
     )
    )
    Particles.FlareSize:Vector2f=&Multiply_Vector2D_ByFloat(
     Float=&FloatFromCurve(
      FloatCurve=DataInterface() {propertyValues {Curve {keys [{arriveTangent=1.5285, interpMode=RCIM_Cubic, leaveTangent=1.5285, tangentMode=RCTM_User, time=0, value=0}, <at 1 1>]}}}
      Scale Curve=1.2
     )
     Vector2D=$Particles.Initial.SpriteSize
    )
    Particles.FlashCameraOffset=8
   ) {isSetParameters=true}
  ]
  rendererClasses [&NiagaraSpriteRendererProperties]
  renderers [
   {rendererClass=&NiagaraSpriteRendererProperties}
   {rendererClass=&NiagaraSpriteRendererProperties, rendererIndex=1}
  ]
 }
}
summary {
 systemName=NS_Emote_ModernMix_HandsBurst
 emitters [
  {emitterName=HandBurst, rendererClasses [&NiagaraSpriteRendererProperties]}
  {emitterName=WristAirglow, rendererClasses [&NiagaraSpriteRendererProperties]}
 ]
 userVariables [
  User.bIsFrontEnd {type=NiagaraBool}
  User.bIsFrontEndPreview {type=NiagaraBool}
  User.RibbonIDOffset {type=NiagaraInt32}
  User.Scale {type=NiagaraFloat, defaultValue=1, description=""}
 ]
}
