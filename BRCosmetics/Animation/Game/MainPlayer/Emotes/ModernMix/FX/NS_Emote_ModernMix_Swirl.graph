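# Niagara effect for the ModernMix emote that plays a warm, pale swirl of spinning meshes with a ground wave.
# GroundWave bursts at the base, MeshSwirls lift toward a loop-driven height, and WindBurst1 adds growing, rotating wind meshes; WindBurst2 is disabled.
# Curves over particle and loop age drive scale, spin, fades and dynamic material parameters.
# Time-of-day functions adjust tint and opacity, and GroundWave's alpha also differs between front end and gameplay.
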NiagaraSystem /BRCosmetics/Animation/Game/MainPlayer/Emotes/ModernMix/FX/NS_Emote_ModernMix_Swirl
dependencies {
 usedDataInterfaces [
  &NiagaraDataInterfaceArrayColor
  &NiagaraDataInterfaceCurve
  &NiagaraDataInterfaceColorCurve
  &NiagaraDataInterfaceVectorCurve
  &NiagaraDataInterfaceSkeletalMesh
 ]
 usedDynamicInputs [&MakeVector4_FromLinearColor, &ColorFromCurve]
 usedModules [
  &self:SystemSpawnScript.SystemScriptSource.SystemScriptGraph.NiagaraNodeAssignment_1.SetVariables_#1
  &SystemState
 ]
}
emitters {
 GroundWave {
  emitterSpawnScript [/SetVariables(
    Emitter.DarkColor:LinearColor=&FNTimeOfDayColor(
     Night Color=<rgba 0.1875 0.1208 0.0947 1>
     Morning Color=<rgba 0.3542 0.211 0.1549 1>
     Day Color=<rgba 0.4062 0.2511 0.1904 1>
     Evening Color=<rgba 0.4062 0.2511 0.1904 1>
    )
    Emitter.MainColor:LinearColor=&SelectLinearColorFromArray(
     Array Sampling Mode=::NewEnumerator0
     Direct Array Index=&RandomRangeInt(Minimum=0, Maximum=1)
     Color Selection Array=DataInterface() {propertyValues {ColorData [<rgba 1 0.7257 0.4955 1>, <rgba 1 0.6963 0.5677 1>]}}
     Random Seed=0
    )
   ) {isSetParameters=true}]
  emitterUpdateScript [
   &EmitterState(Life Cycle Mode=::NewEnumerator1, Loop Behavior=::NewEnumerator1, Loop Duration=4, UseLoopDelay=false)
   &SpawnBurst_Instantaneous(Use Spawn Probability=false, UseLoopCountLimit=true)
  ]
  particleSpawnScript [
   &InitializeParticle(
    Color=<rgba 1 1 1 8>
    Lifetime=4
    Mesh Scale Mode=::NewEnumerator3
    Mesh Uniform Scale=0.8
    Position Offset=<0 0 5>
    UsePositionOffset=true
    Write Material Random=true
   )
   &InitialMeshOrientation(
    Mesh Orientation Mode=::NewEnumerator2
    Orientation Vector=<0 0 1>
    Rotation=&RandomRangeVector(Minimum=<-5 0 0>, Maximum=<5 0 0>)
    Use Rotation Vector=true
   )
   &CameraOffset(Camera Offset Amount=5)
  ]
  particleUpdateScript [
   &ParticleState()
   &ScaleColor(
    Scale Alpha=&FloatFromCurve(
     FloatCurve=DataInterface() {propertyValues {Curve {keys [
         {arriveTangent=8.3055, interpMode=RCIM_Cubic, leaveTangent=8.3055, tangentMode=RCTM_User, time=0.1, value=0}
         {interpMode=RCIM_Cubic, tangentMode=RCTM_User, time=0.5, value=1}
         {arriveTangent=-1.2034, tangentMode=RCTM_User, time=1, value=0}
        ]}}}
     Scale Curve=&Multiply_Float(
      A=&FN_FrontEndFloatModifier(Gameplay=0.45, Front End=0.3)
      B=&FNTimeOfDayScalar(Night Scale=0.7, Morning Scale=0.9, Evening Scale=0.8)
     )
    )
    ScaleA=true
    ScaleRGB=true
   ) {module=ScaleColor001}
  ]
  rendererClasses [&NiagaraMeshRendererProperties]
  renderers [{rendererClass=&NiagaraMeshRendererProperties, rendererIndex=0}]
 }
 MeshSwirls {
  emitterSpawnScript [/SetVariables(Emitter.DarkColor:LinearColor=<rgba 0.6458 0.515 0.4171 1>, Emitter.MainColor:LinearColor=<rgba 1 0.7334 0.5156 1>) {isSetParameters=true}]
  emitterUpdateScript [
   &EmitterState(Life Cycle Mode=::NewEnumerator1, Loop Behavior=::NewEnumerator1, Loop Delay=0.15, Loop Duration=3, UseLoopDelay=true)
   &SpawnRate(SpawnRate=3, Use Spawn Probability=false) {module=SpawnRate001}
  ]
  particleSpawnScript [
   &InitializeParticle(
    Color=&SelectLinearColorFromArray(
     Array Sampling Mode=::NewEnumerator0
     Direct Array Index=?
     Color Selection Array=DataInterface() {propertyValues {ColorData [<rgba 1 0.6689 0.3599 1>, <rgba 1 0.7694 0.4481 1>, <rgba 1 0.7142 0.3992 1.5>, <rgba 1 0.7757 0.428 1>, <rgba 1 0.751 0.4016 1>]}}
     Random Seed=0
    )
    Lifetime Max=1.2
    Lifetime Mode=::NewEnumerator1
    Mesh Scale=&Multiply_VectorByFloat(Vector=<1.55 1.55 1.2>, Float=&RandomRangeFloat(Minimum=1, Maximum=1.2))
    Mesh Scale Mode=::NewEnumerator4
    Position Offset=<0 0 200>
    UsePositionOffset=false
    Write Material Random=true
   )
   &InitialMeshOrientation(
    Mesh Orientation Mode=::NewEnumerator7
    Rotation=&RandomRangeVector(Minimum=<0 0 -1>, Maximum=<0 0 1>)
    Use Rotation Vector=true
   )
   &ShapeLocation(
    Cylinder Height=0
    Cylinder Height Midpoint=0
    Cylinder Radius=0
    MaskSpawnGroup=false
    Non Uniform Scale=&Multiply_VectorByFloat(Vector=<1 1 1>, Float=$User.Scale)
    Override Seed=false
    Shape Primitive=::NewEnumerator1
    Surface Only=false
    UseOwnerScale=true
   )
   /SetVariables(Particles.ZGoalpost:NiagaraFloat=&FloatFromCurve(
     FloatCurve=DataInterface() {propertyValues {Curve {keys [
         {arriveTangent=6.3086, interpMode=RCIM_Cubic, leaveTangent=6.3086, tangentMode=RCTM_User, time=0, value=50}
         {tangentMode=RCTM_User, time=1, value=65}
        ]}}}
     CurveIndex=$Emitter.NormalizedLoopAge
     Scale Curve=1.3
    )) {isSetParameters=true}
  ]
  particleUpdateScript [
   &ParticleState()
   /SetVariables(Particles.Position=&Lerp_Vector(
     A=Linked() {linkedVariable {name=Particles.Initial.Position, type=NiagaraPosition}}
     B=&MakeVector(X=0, Y=0, Z=$Particles.Initial.ZGoalpost)
     Alpha=&FloatFromCurve(FloatCurve=DataInterface() {propertyValues {Curve {keys [
          {arriveTangent=1.2875, interpMode=RCIM_Cubic, leaveTangent=1.2875, tangentMode=RCTM_User, time=0, value=0}
          {tangentMode=RCTM_User, time=1, value=1}
         ]}}})
    )) {isSetParameters=true}
   &DynamicMaterialParameters(
    Index 0 Param 1=&FloatFromCurve(FloatCurve=DataInterface() {propertyValues {Curve {keys [{interpMode=RCIM_Cubic, tangentMode=RCTM_User, time=0.5, value=0}, {interpMode=RCIM_Cubic, tangentMode=RCTM_User, time=1, value=1}]}}})
    Index 0 Param 3=1
    Module.Index0Float0Write=true
    Module.Index0Float1Write=true
    Module.Index0Float2Write=true
    Module.Index0Float3Write=true
   )
   &ScaleMeshSize(Scale Factor=&Multiply_VectorByFloat(
     Vector=<1 1 1.5>
     Float=&FloatFromCurve(
      FloatCurve=DataInterface() {propertyValues {Curve {keys [
          {arriveTangent=1.1179, interpMode=RCIM_Cubic, leaveTangent=1.1179, tangentMode=RCTM_User, time=0, value=0.35}
          {arriveTangent=0.0224, interpMode=RCIM_Cubic, leaveTangent=0.0224, tangentMode=RCTM_User, time=1, value=1}
         ]}}}
      Scale Curve=&Multiply_Float(A=1, B=1)
     )
    ))
   &UpdateMeshOrientation(
    Rotation Rate=&FloatFromCurve(FloatCurve=DataInterface() {propertyValues {Curve {keys [<at 0 0>,<at 1 1>]}}}, Scale Curve=-2)
    Rotation Vector=<0 0 1>
   )
   &ScaleColor(
    Scale Alpha=&FloatFromCurve(
     FloatCurve=DataInterface() {propertyValues {Curve {keys [
         {arriveTangent=1.8077, interpMode=RCIM_Cubic, leaveTangent=1.8077, tangentMode=RCTM_User, time=0, value=0}
         {arriveTangent=-1.9397, interpMode=RCIM_Cubic, leaveTangent=-2.9533, tangentMode=RCTM_Break, time=0.5, value=0.5}
         {arriveTangent=-0.7636, interpMode=RCIM_Cubic, leaveTangent=-0.7636, tangentMode=RCTM_User, time=1, value=0}
        ]}}}
     Scale Curve=&Multiply_Float(A=0.408, B=&FNTimeOfDayScalar(Night Scale=0.6, Morning Scale=0.9, Evening Scale=0.8))
    )
    ScaleA=true
    ScaleRGB=true
   )
   &ScaleColor(
    Scale Alpha=&FloatFromCurve(FloatCurve=DataInterface() {propertyValues {Curve {keys [<at 0.8 1>,<at 0.93 0>]}}}, CurveIndex=$System.NormalizedLoopAge)
    ScaleA=true
    ScaleRGB=true
   ) {module=ScaleColor001}
  ]
  rendererClasses [&NiagaraMeshRendererProperties]
  renderers [{rendererClass=&NiagaraMeshRendererProperties, rendererIndex=0}]
 }
 WindBurst1 {
  emitterSpawnScript [
   /SetVariables(Emitter.RotationSpeed:NiagaraFloat=&FloatFromCurve(FloatCurve=DataInterface() {propertyValues {Curve {keys [<at 0 -0.5>,<at 0.1 0>]}}}, CurveIndex=$Emitter.NormalizedLoopAge, Scale Curve=0)) {isSetParameters=true, module=SetVariables_#2}
   /SetVariables(
    Emitter.DarkColor:LinearColor=&FNTimeOfDayColor(
     Night Color=<rgba 0.6458 0.5867 0.5718 1>
     Morning Color=<rgba 0.6458 0.5943 0.5651 1>
     Day Color=<rgba 0.6458 0.5892 0.5449 1>
     Evening Color=<rgba 0.6458 0.6102 0.5819 1>
    )
    Emitter.MainColor:LinearColor=&FNTimeOfDayColor(
     Night Color=&SelectLinearColorFromArray(
      Array Sampling Mode=::NewEnumerator0
      Direct Array Index=&RandomRangeInt(Minimum=0, Maximum=1)
      Color Selection Array=DataInterface() {propertyValues {ColorData [<rgba 1 0.947 0.8646 1>]}}
      Random Seed=0
     )
     Morning Color=&SelectLinearColorFromArray(
      Array Sampling Mode=::NewEnumerator0
      Direct Array Index=&RandomRangeInt(Minimum=0, Maximum=1)
      Color Selection Array=DataInterface() {propertyValues {ColorData [<rgba 1 0.8815 0.8698 1>]}}
      Random Seed=0
     )
     Day Color=&SelectLinearColorFromArray(
      Array Sampling Mode=::NewEnumerator0
      Direct Array Index=&RandomRangeInt(Minimum=0, Maximum=1)
      Color Selection Array=DataInterface() {propertyValues {ColorData [<rgba 1 0.9425 0.901 1>, <rgba 1 0.9106 0.8802 1>]}}
      Random Seed=0
     )
     Evening Color=&SelectLinearColorFromArray(
      Array Sampling Mode=::NewEnumerator0
      Direct Array Index=&RandomRangeInt(Minimum=0, Maximum=1)
      Color Selection Array=DataInterface() {propertyValues {ColorData [<rgba 1 0.9332 0.8802 1>]}}
      Random Seed=0
     )
    )
   ) {isSetParameters=true, module=SetVariables_#3}
  ]
  emitterUpdateScript [
   &EmitterState(Life Cycle Mode=::NewEnumerator1, Loop Behavior=::NewEnumerator1, UseLoopDelay=false)
   &SpawnRate(
    SpawnRate=&FloatFromCurve(
     FloatCurve=DataInterface() {propertyValues {Curve {keys [<at 0 0.6>,<at 0.6096 1>,<at 1 0.5295>]}}}
     CurveIndex=$Emitter.NormalizedLoopAge
     Scale Curve=7
    )
    Use Spawn Probability=true
   )
  ]
  particleSpawnScript [
   &InitializeParticle(
    Color=<rgba 1 1 1 3>
    Lifetime=&RandomRangeFloat(Minimum=1, Maximum=1.3)
    Mesh Scale=&Multiply_VectorByFloat(Vector=<2 2 1.0755>, Float=&RandomRangeFloat(Minimum=0.5, Maximum=0.65))
    Mesh Scale Mode=::NewEnumerator4
    Position Offset Coordinate Space=::Local
    UsePositionOffset=true
    Write Material Random=true
   )
   &InitialMeshOrientation(
    Mesh Axis To Orient=<0 0 3>
    Mesh Orientation Mode=::NewEnumerator2
    Orientation Coordinate Space=::World
    Orientation Vector=<0 0 1>
    Rotation=&RandomRangeVector(Minimum=<0.03 0.03 0>, Maximum=<-0.03 -0.03 1>)
    Use Rotation Vector=true
   )
  ]
  particleUpdateScript [
   &ParticleState()
   &ScaleMeshSize(Scale Factor=&VectorFromCurve(VectorCurve=DataInterface() {propertyValues {
       XCurve {keys [<at 0 1>,<at 1 1>]}
       YCurve {keys [<at 0 1>,<at 1 1>]}
       ZCurve {keys [<at 0 0.5 RCIM_Cubic>, {arriveTangent=1.1019, tangentMode=RCTM_User, time=1, value=1.4}]}
      }})) {module=ScaleMeshSize002}
   &ScaleMeshSize(Scale Factor=&VectorFromFloat(Value=&FloatFromCurve(FloatCurve=DataInterface() {propertyValues {Curve {keys [<at 0 0 RCIM_Cubic>,<at 1 1>]}}})))
   &UpdateMeshOrientation(
    Rotation Rate=&FloatFromCurve(
     FloatCurve=DataInterface() {propertyValues {Curve {keys [<at 0.55 0.4376 RCIM_Cubic>, {arriveTangent=2.9123, interpMode=RCIM_Cubic, leaveTangent=2.9123, tangentMode=RCTM_User, time=1, value=1}]}}}
     CurveIndex=$Emitter.NormalizedLoopAge
     Scale Curve=-1.7518
    )
    Rotation Vector=<0 0 1>
   )
   &ScaleColor(
    Scale Alpha=&FloatFromCurve(
     FloatCurve=DataInterface() {propertyValues {Curve {keys [
         <at 0 0 RCIM_Cubic>
         {interpMode=RCIM_Cubic, tangentMode=RCTM_User, time=0.2226, value=1}
         {interpMode=RCIM_Cubic, tangentMode=RCTM_User, time=0.5806, value=1}
         {tangentMode=RCTM_User, time=1, value=0}
        ]}}}
     Scale Curve=&Multiply_Float(A=0.2, B=&FNTimeOfDayScalar(Night Scale=0.7, Morning Scale=0.9, Evening Scale=0.8))
    )
    Scale RGB=<2 2 2>
    ScaleA=true
    ScaleRGB=true
   )
   &DynamicMaterialParameters(
    Index 0 Param 1=&FloatFromCurve(FloatCurve=DataInterface() {propertyValues {Curve {keys [{tangentMode=RCTM_User, time=0.4, value=1}, {tangentMode=RCTM_User, time=0.9, value=-1}]}}})
    Index 0 Param 2=1
    Index 0 Param 3=&RandomRangeFloat(Minimum=-1, Maximum=1)
    Index 0 Param 4=&RandomRangeFloat(Minimum=-1, Maximum=1)
    Index 1 Param 1=&FloatFromCurve(
     FloatCurve=DataInterface() {propertyValues {Curve {keys [
         {arriveTangent=1.7366, interpMode=RCIM_Cubic, leaveTangent=1.7366, tangentMode=RCTM_User, time=0, value=0}
         {arriveTangent=0.4174, interpMode=RCIM_Cubic, leaveTangent=0.4174, tangentMode=RCTM_User, time=1, value=1}
        ]}}}
     Scale Curve=0.5
    )
    Index 1 Param 2=1
    Module.Index0Float0Write=true
    Module.Index0Float1Write=true
    Module.Index0Float2Write=true
    Module.Index0Float3Write=true
    Module.Index1Float0Write=true
    Module.Index1Float1Write=false
    Module.Index1Float2Write=false
    Module.Index1Float3Write=false
    Write Parameter Index 1=::NewEnumerator1
   ) {module=DynamicMaterialParameters001}
   &ScaleColor(
    Scale Alpha=&FloatFromCurve(
     FloatCurve=DataInterface() {propertyValues {Curve {keys [<at 0.1 0>, <at 0.2 1>, {arriveTangent=-1.5385, interpMode=RCIM_Cubic, leaveTangent=-1.5385, time=0.658, value=1}, <at 0.85 0>]}}}
     CurveIndex=$Emitter.NormalizedLoopAge
    )
    Scale RGB=&VectorFromFloat(Value=&FloatFromCurve(
      FloatCurve=DataInterface() {propertyValues {Curve {keys [<at 0.1 0>,<at 0.2 1>,<at 0.7 1>,<at 0.9 0>]}}}
      CurveIndex=$Emitter.NormalizedLoopAge
     ))
    ScaleA=true
    ScaleRGB=true
   ) {module=ScaleColor001}
  ]
  rendererClasses [&NiagaraMeshRendererProperties]
  renderers [{rendererClass=&NiagaraMeshRendererProperties, rendererIndex=0}]
 }
 WindBurst2 {
  bEnabled=false
  emitterSpawnScript [
   /SetVariables(Emitter.RotationSpeed:NiagaraFloat=&FloatFromCurve(FloatCurve=DataInterface() {propertyValues {Curve {keys [<at 0 -0.5>,<at 0.1 0>]}}}, CurveIndex=$Emitter.NormalizedLoopAge, Scale Curve=0)) {isSetParameters=true, module=SetVariables_#2}
   /SetVariables(
    Emitter.DarkColor:LinearColor=&FNTimeOfDayColor(
     Night Color=<rgba 0.6458 0.5867 0.5718 1>
     Morning Color=<rgba 0.6458 0.5943 0.5651 1>
     Day Color=<rgba 0.6458 0.5892 0.5449 1>
     Evening Color=<rgba 0.6458 0.6102 0.5819 1>
    )
    Emitter.MainColor:LinearColor=&FNTimeOfDayColor(
     Night Color=&SelectLinearColorFromArray(
      Array Sampling Mode=::NewEnumerator0
      Direct Array Index=&RandomRangeInt(Minimum=0, Maximum=1)
      Color Selection Array=DataInterface() {propertyValues {ColorData [<rgba 1 0.947 0.8646 1>]}}
      Random Seed=0
     )
     Morning Color=&SelectLinearColorFromArray(
      Array Sampling Mode=::NewEnumerator0
      Direct Array Index=&RandomRangeInt(Minimum=0, Maximum=1)
      Color Selection Array=DataInterface() {propertyValues {ColorData [<rgba 1 0.8815 0.8698 1>]}}
      Random Seed=0
     )
     Day Color=&SelectLinearColorFromArray(
      Array Sampling Mode=::NewEnumerator0
      Direct Array Index=&RandomRangeInt(Minimum=0, Maximum=1)
      Color Selection Array=DataInterface() {propertyValues {ColorData [<rgba 1 0.9425 0.901 1>, <rgba 1 0.9106 0.8802 1>]}}
      Random Seed=0
     )
     Evening Color=&SelectLinearColorFromArray(
      Array Sampling Mode=::NewEnumerator0
      Direct Array Index=&RandomRangeInt(Minimum=0, Maximum=1)
      Color Selection Array=DataInterface() {propertyValues {ColorData [<rgba 1 0.9332 0.8802 1>]}}
      Random Seed=0
     )
    )
   ) {isSetParameters=true, module=SetVariables_#3}
  ]
  emitterUpdateScript [
   &EmitterState(Life Cycle Mode=::NewEnumerator1, Loop Behavior=::NewEnumerator1, Loop Delay=2.14, Loop Duration=1.2, UseLoopDelay=true)
   &SpawnRate(SpawnRate=3.6, Use Spawn Probability=true)
  ]
  particleSpawnScript [
   &InitializeParticle(
    Color=<rgba 1 1 1 3>
    Lifetime=&RandomRangeFloat(Minimum=1, Maximum=1.3)
    Mesh Scale=&Multiply_VectorByFloat(Vector=<3.6 3.6 0.5>, Float=&RandomRangeFloat(Minimum=0.5, Maximum=0.65))
    Mesh Scale Mode=::NewEnumerator4
    Position Offset=<0 0 50>
    Position Offset Coordinate Space=::Local
    UsePositionOffset=true
    Write Material Random=true
   )
   &InitialMeshOrientation(
    Mesh Axis To Orient=<0 0 3>
    Mesh Orientation Mode=::NewEnumerator2
    Orientation Coordinate Space=::World
    Orientation Vector=<0 0 1>
    Rotation=&RandomRangeVector(Minimum=<0.03 0.03 0>, Maximum=<-0.03 -0.03 1>)
    Use Rotation Vector=true
   )
  ]
  particleUpdateScript [
   &ParticleState()
   &ScaleMeshSize(Scale Factor=&VectorFromFloat(Value=&FloatFromCurve(FloatCurve=DataInterface() {propertyValues {Curve {keys [<at 0 0.4137 RCIM_Cubic>,<at 1 0.7306>]}}})))
   &UpdateMeshOrientation(
    Rotation Rate=&FloatFromCurve(
     FloatCurve=DataInterface() {propertyValues {Curve {keys [<at 0.55 0.4376 RCIM_Cubic>, {arriveTangent=2.9123, interpMode=RCIM_Cubic, leaveTangent=2.9123, tangentMode=RCTM_User, time=1, value=1}]}}}
     CurveIndex=$Emitter.NormalizedLoopAge
     Scale Curve=-1.6154
    )
    Rotation Vector=<0 0 1>
   )
   &ScaleColor(
    Scale Alpha=&FloatFromCurve(
     FloatCurve=DataInterface() {propertyValues {Curve {keys [
         <at 0 0 RCIM_Cubic>
         {interpMode=RCIM_Cubic, tangentMode=RCTM_User, time=0.3258, value=1}
         {interpMode=RCIM_Cubic, tangentMode=RCTM_User, time=0.5806, value=1}
         {tangentMode=RCTM_User, time=1, value=0}
        ]}}}
     Scale Curve=&Multiply_Float(A=0.09, B=&FNTimeOfDayScalar(Night Scale=0.7, Morning Scale=0.9, Evening Scale=0.8))
    )
    Scale RGB=<2 2 2>
    ScaleA=true
    ScaleRGB=true
   )
   &DynamicMaterialParameters(
    Index 0 Param 1=&FloatFromCurve(FloatCurve=DataInterface() {propertyValues {Curve {keys [{tangentMode=RCTM_User, time=0.4, value=1}, {tangentMode=RCTM_User, time=0.9, value=-1}]}}})
    Index 0 Param 2=1
    Index 0 Param 3=&RandomRangeFloat(Minimum=-1, Maximum=1)
    Index 0 Param 4=&RandomRangeFloat(Minimum=-1, Maximum=1)
    Index 1 Param 1=&FloatFromCurve(
     FloatCurve=DataInterface() {propertyValues {Curve {keys [
         {arriveTangent=1.7366, interpMode=RCIM_Cubic, leaveTangent=1.7366, tangentMode=RCTM_User, time=0, value=0}
         {arriveTangent=0.4174, interpMode=RCIM_Cubic, leaveTangent=0.4174, tangentMode=RCTM_User, time=1, value=1}
        ]}}}
     Scale Curve=0.5
    )
    Index 1 Param 2=1
    Module.Index0Float0Write=true
    Module.Index0Float1Write=true
    Module.Index0Float2Write=true
    Module.Index0Float3Write=true
    Module.Index1Float0Write=true
    Module.Index1Float1Write=false
    Module.Index1Float2Write=false
    Module.Index1Float3Write=false
    Write Parameter Index 1=::NewEnumerator1
   ) {module=DynamicMaterialParameters001}
  ]
  rendererClasses [&NiagaraMeshRendererProperties]
  renderers [{rendererClass=&NiagaraMeshRendererProperties, rendererIndex=0}]
 }
}
summary {
 systemName=NS_Emote_ModernMix_Swirl
 emitters [
  {emitterName=MeshSwirls, rendererClasses [&NiagaraMeshRendererProperties]}
  {emitterName=GroundWave, rendererClasses [&NiagaraMeshRendererProperties]}
  {emitterName=WindBurst1, rendererClasses [&NiagaraMeshRendererProperties]}
  {bEnabled=false, emitterName=WindBurst2, rendererClasses [&NiagaraMeshRendererProperties]}
 ]
 userVariables [
  User.bIsFrontEnd {type=NiagaraBool}
  User.bIsFrontEndPreview {type=NiagaraBool}
  User.RibbonIDOffset {type=NiagaraInt32}
  User.Scale {type=NiagaraFloat, defaultValue=1, description=""}
 ]
}
