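# Niagara system for the ModernMix emote that adds warm glowing trails and glows to the character's hands and wrists.
# Two ribbon emitters draw hand and bracelet anim trails from skeletal mesh bones or sockets, with curves over normalized loop age timing their visibility.
# HandGlow adds pulsing attached sprites, while WristAirglow adds sprites and particle lights.
# Front-end modifiers adjust all emitters' intensity, and time-of-day scalars also dim trails and airglow.
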NiagaraSystem /BRCosmetics/Animation/Game/MainPlayer/Emotes/ModernMix/FX/NS_Emote_ModernMix_Trails
dependencies {
 usedDataInterfaces [&NiagaraDataInterfaceCurve, &NiagaraDataInterfaceColorCurve]
 usedDynamicInputs [&SimulationPosition]
 usedModules [
  &self:SystemSpawnScript.SystemScriptSource.SystemScriptGraph.NiagaraNodeAssignment_1.SetVariables_#1
  &SystemState
 ]
}
emitters {
 BraceletAnimTrail {
  emitterUpdateScript [
   &EmitterState(Life Cycle Mode=::NewEnumerator1, Loop Behavior=::NewEnumerator1, Loop Duration Mode=::NewEnumerator1, UseLoopDelay=false)
   &FN_SpawnAnimTrail_v2(Number of Anim Trails=2, Spawn Interval=0, Spawn Mode:ENiagaraSkeletalSpawning=::NewEnumerator0)
  ]
  particleSpawnScript [
   /Game/Effects/Niagara/Modules/AnimTrail/FN_InitializeAnimTrail/FN_InitializeAnimTrailBoneIndecies(First Bone or Socket Index=2, Second Bone or Socket Index=3)
   &FN_InitializeAnimTrailBoneIndecies(AnimTrail Selection=1, First Bone or Socket Index=4, Second Bone or Socket Index=5) {module=FN_InitializeAnimTrailBoneIndecies001}
   &FN_InitializeAnimTrail(AnimTrailWidth=2, Color=<rgba 1 0.6332 0.4948 0.68>, SkeletalMesh=$System.Skeletal Mesh)
  ]
  particleUpdateScript [
   /Game/Effects/Niagara/Modules/AnimTrail/FN_UpdateAnimTrail/FN_UpdateAnimTrail_v4/FN_UpdateAnimTrail()
   &InheritSourceMovement(ApplyForce=false, ApplyPosition=true, ApplyPositionFalloff=false)
   &ScaleColor(
    Scale Alpha=&FloatFromCurve(FloatCurve=DataInterface() {propertyValues {Curve {keys [
         {interpMode=RCIM_Cubic, tangentMode=RCTM_User, time=0, value=1}
         {arriveTangent=-2.7485, interpMode=RCIM_Cubic, leaveTangent=-2.7485, tangentMode=RCTM_User, time=1, value=0}
        ]}}})
    ScaleA=true
    ScaleRGB=true
   )
   &SolveForcesAndVelocity(Clamp Velocity=false, Limit Acceleration=false)
   &ScaleColor(
    Scale Alpha=&FloatFromCurve(
     FloatCurve=DataInterface() {propertyValues {Curve {keys [
         <at 0.2299 0 RCIM_Cubic>
         <at 0.3446 1>
         {arriveTangent=-0.3163, interpMode=RCIM_Cubic, leaveTangent=-0.3163, tangentMode=RCTM_User, time=0.3955, value=1}
         {arriveTangent=-4.8898, leaveTangent=-4.8898, time=0.43, value=0}
         {arriveTangent=5.1677, leaveTangent=5.1677, time=0.6, value=0}
         {arriveTangent=0.115, interpMode=RCIM_Cubic, leaveTangent=0.115, tangentMode=RCTM_User, time=0.6363, value=1}
         {arriveTangent=-0.4222, interpMode=RCIM_Cubic, leaveTangent=-0.4222, tangentMode=RCTM_User, time=0.9, value=1}
         <at 0.95 0 RCIM_Cubic>
        ]}}}
     CurveIndex=$System.NormalizedLoopAge
     Scale Curve=&Multiply_Float(
      A=&FN_FrontEndFloatModifier(Gameplay=1, Front End=0.8)
      B=&FNTimeOfDayScalar(Night Scale=0.6, Morning Scale=0.9, Evening Scale=0.8)
     )
    )
    ScaleA=true
    ScaleRGB=true
   ) {module=ScaleColor001}
  ]
  rendererClasses [&NiagaraRibbonRendererProperties]
  renderers [{rendererClass=&NiagaraRibbonRendererProperties}]
 }
 HandGlow {
  emitterUpdateScript [
   &EmitterState(Life Cycle Mode=::NewEnumerator1, Loop Behavior=::NewEnumerator1, Loop Duration Mode=::NewEnumerator1, UseLoopDelay=false)
   &SpawnBurst_Instantaneous(Spawn Count=2, Use Spawn Probability=false)
  ]
  particleSpawnScript [
   &FN_SetSpawnGroup(Number of Spawn Groups=2)
   &InitializeParticle(
    Color=<rgba 1.5 1.1052 0.7771 0.244>
    Lifetime=10.9
    Mass Mode=::NewEnumerator1
    Sprite Size Mode=::NewEnumerator3
    Uniform Sprite Size=40
    UsePositionOffset=false
    Write Material Random=true
   ) {module=InitializeParticle001}
   &CameraOffset(Camera Offset Amount=1)
  ]
  particleUpdateScript [
   &ParticleState()
   &ScaleSpriteSize(
    Uniform Curve Scale=&Sine(Normalized Angle=&Add_Float(A=$Emitter.Age, B=0), Period=0.8, Scale=0.1, Bias=0.8)
    Uniform Curve Sprite Scale=DataInterface() {propertyValues {Curve {keys [{tangentMode=RCTM_User, time=0, value=1}, {tangentMode=RCTM_User, time=1, value=1}]}}}
   )
   &SkeletalMeshLocation(
    Bone / Socket Index=$Particles.SpawnGroup
    MaskSpawnGroup=false
    Mesh Sampling Type=::NewEnumerator5
    Skeletal Mesh=$System.Skeletal Mesh
    Socket Sampling Mode=::NewEnumerator5
   )
   /SetVariables(Particles.Color_Decal:LinearColor=&ColorFromCurve(ColorCurve=DataInterface() {propertyValues {
       AlphaCurve {keys [<at 0 0 RCIM_Cubic>,<at 0.05 0.1>,<at 0.9 0.1>,<at 1 0 RCIM_Cubic>]}
       BlueCurve {keys [<at 0 0.5556 RCIM_Cubic>]}
       GreenCurve {keys [<at 0 0.7237 RCIM_Cubic>]}
       RedCurve {keys [<at 0 1 RCIM_Cubic>]}
      }})) {isSetParameters=true}
   &ScaleColor(
    Scale Alpha=&FloatFromCurve(
     FloatCurve=DataInterface() {propertyValues {Curve {keys [
         <at 0 0 RCIM_Cubic>
         {arriveTangent=0.1666, interpMode=RCIM_Cubic, leaveTangent=0.1666, tangentMode=RCTM_User, time=0.1, value=1}
         {arriveTangent=0.047, interpMode=RCIM_Cubic, leaveTangent=0.047, tangentMode=RCTM_User, time=0.9, value=1}
         <at 1 0 RCIM_Cubic>
        ]}}}
     Scale Curve=&FN_FrontEndFloatModifier(Gameplay=1, Front End=0.75)
    )
    ScaleA=true
    ScaleRGB=true
   ) {module=ScaleColor002}
  ]
  rendererClasses [&NiagaraSpriteRendererProperties]
  renderers [{rendererClass=&NiagaraSpriteRendererProperties}]
 }
 HandsRibbonMain {
  emitterUpdateScript [
   &EmitterState(Life Cycle Mode=::NewEnumerator1, Loop Behavior=::NewEnumerator1, Loop Duration Mode=::NewEnumerator1, UseLoopDelay=true)
   &FN_SpawnAnimTrail_v2(Number of Anim Trails=2, Spawn Interval=0, Spawn Mode:ENiagaraSkeletalSpawning=::NewEnumerator0)
  ]
  particleSpawnScript [
   &FN_InitializeAnimTrailBoneIndecies()
   &FN_InitializeAnimTrailBoneIndecies(AnimTrail Selection=1, First Bone or Socket Index=1, Second Bone or Socket Index=1) {module=FN_InitializeAnimTrailBoneIndecies001}
   &FN_InitializeAnimTrail(AnimTrailWidth=30, Color=<rgba 1 1 1 0.04>, Lifetime=1.1, SkeletalMesh=$System.Skeletal Mesh)
  ]
  particleUpdateScript [
   &FN_UpdateAnimTrail()
   &InheritSourceMovement(ApplyForce=false, ApplyPosition=true, ApplyPositionFalloff=false)
   &ParticleState() {enabled=false}
   &ScaleColor(
    Linear Color Curve=DataInterface() {propertyValues {
      AlphaCurve {keys [
        <at 0 1 RCIM_Cubic>
        {arriveTangent=-2.9877, interpMode=RCIM_Cubic, leaveTangent=-2.9877, tangentMode=RCTM_User, time=0.2319, value=2}
        {arriveTangent=-0.0599, interpMode=RCIM_Cubic, leaveTangent=-0.0599, tangentMode=RCTM_User, time=0.9326, value=0}
       ]}
      BlueCurve {keys [<at 0 4.3596>,<at 0.1168 1.4297>,<at 0.8124 0.1199>]}
      GreenCurve {keys [<at 0 12.6357>,<at 0.1168 3.0649>,<at 0.8124 0.3649>]}
      RedCurve {keys [<at 0 20>,<at 0.1168 5>,<at 0.8124 1>]}
     }}
    Scale Mode=::NewEnumerator2
   )
   &ScaleRibbonWidth(Ribbon Width Scale=&FloatFromCurve(FloatCurve=DataInterface() {propertyValues {Curve {keys [<at 0 1 RCIM_Cubic>, {tangentMode=RCTM_User, time=1, value=0.357}]}}}))
   &DynamicMaterialParameters(
    Index 0 Param 1=&FloatFromCurve(FloatCurve=DataInterface() {propertyValues {Curve {keys [<at 0.1 1>,<at 1 0>]}}})
    Index 0 Param 2=&FloatFromCurve(FloatCurve=DataInterface() {propertyValues {Curve {keys [
         {arriveTangent=1.2896, interpMode=RCIM_Cubic, leaveTangent=1.2896, tangentMode=RCTM_User, time=0, value=0}
         <at 0.1 1 RCIM_Cubic>
         <at 0.5 0 RCIM_Cubic>
        ]}}})
    Index 0 Param 3=-1
    Index 0 Param 4=2
    Module.Param0WriteEnabled=true
    Module.Param1WriteEnabled=true
    Module.Param2WriteEnabled=true
    Module.Param3WriteEnabled=true
   ) {module=DynamicMaterialParameters001}
   &ScaleColor(
    Scale Alpha=&FloatFromCurve(
     FloatCurve=DataInterface() {propertyValues {Curve {keys [
         <at 0.3 0 RCIM_Cubic>
         <at 0.3446 1>
         {arriveTangent=-11.7117, interpMode=RCIM_Cubic, leaveTangent=-11.7117, time=0.3955, value=1}
         <at 0.43 0>
         <at 0.6 0>
         {arriveTangent=-0.2186, interpMode=RCIM_Cubic, leaveTangent=-0.2186, tangentMode=RCTM_User, time=0.6289, value=1}
         {arriveTangent=0.0431, interpMode=RCIM_Cubic, leaveTangent=0.0431, tangentMode=RCTM_User, time=0.9, value=1}
         <at 0.95 0 RCIM_Cubic>
        ]}}}
     CurveIndex=$System.NormalizedLoopAge
     Scale Curve=&Multiply_Float(
      A=&FN_FrontEndFloatModifier(Gameplay=1, Front End=0.8)
      B=&FNTimeOfDayScalar(Night Scale=0.6, Morning Scale=0.9, Evening Scale=0.8)
     )
    )
    ScaleA=true
    ScaleRGB=true
   ) {module=ScaleColor001}
   &SolveForcesAndVelocity(Clamp Velocity=false, Limit Acceleration=false)
   /SetVariables(
    Particles.Width1:NiagaraFloat=&Multiply_Float(
     A=$Particles.RibbonWidth
     B=&FloatFromCurve(
      FloatCurve=DataInterface() {propertyValues {Curve {keys [{arriveTangent=-2, interpMode=RCIM_Cubic, time=0.52, value=1}, <at 0.55 0 RCIM_Cubic>]}}}
      CurveIndex=$System.NormalizedLoopAge
      Scale Curve=1.2
     )
    )
    Particles.Width2:NiagaraFloat=&Multiply_Float(
     A=$Particles.RibbonWidth
     B=&FloatFromCurve(
      FloatCurve=DataInterface() {propertyValues {Curve {keys [<at 0.58 0 RCIM_Cubic>,<at 0.62 1 RCIM_Cubic>]}}}
      CurveIndex=$System.NormalizedLoopAge
     )
    )
   ) {isSetParameters=true}
  ]
  rendererClasses [&NiagaraRibbonRendererProperties]
  renderers [
   {rendererClass=&NiagaraRibbonRendererProperties}
   {rendererClass=&NiagaraRibbonRendererProperties, rendererIndex=1}
  ]
 }
 WristAirglow {
  emitterUpdateScript [
   &EmitterState(Inactive Response=::NewEnumerator1, Life Cycle Mode=::NewEnumerator1, Loop Behavior=::NewEnumerator1, Loop Duration=10.91, UseLoopDelay=true)
   &SpawnBurst_Instantaneous(Use Spawn Probability=false, UseLoopCountLimit=false)
   &SpawnBurst_Instantaneous(Spawn Group=1, Use Spawn Probability=false, UseLoopCountLimit=false) {module=SpawnBurst_Instantaneous002}
  ]
  particleSpawnScript [
   &InitializeParticle(
    Color=<rgba 1 0.7036 0.4348 0.5>
    Lifetime=&MergeHelper_FloatSwitch(0=10.9, 1=10.9, 2=0.33, 3=0.3, Index=&SpawnGroup(), Num=::NewEnumerator2)
    Sprite Size Mode=::NewEnumerator3
    Uniform Sprite Size=&MergeHelper_FloatSwitch(0=120, 1=120, 2=340, 3=315, Index=&SpawnGroup(), Num=::NewEnumerator2)
    UsePositionOffset=false
    Write Material Random=true
   ) {module=InitializeParticle001}
   &SpriteFacingAndAlignment(Alignment Coordinate Space=::Local, Sprite Alignment=<1 0.25 0>, WriteAlignment=true, WriteFacing=false)
   &CameraOffset(Camera Offset Amount=&MergeHelper_FloatSwitch(0=-2, 1=-2, 2=8, 3=8, Index=&SpawnGroup(), Num=::NewEnumerator2))
   /SetVariables(Particles.FlareVisibility:NiagaraInt32=&SpawnGroup()) {isSetParameters=true}
  ]
  particleUpdateScript [
   &SkeletalMeshLocation(
    Bone / Socket Index=&MergeHelper_IntSwitch(0=0, 1=1, 2=4, 3=4, 4=?, 5=?, Index=&SpawnGroup(), Num=::NewEnumerator2)
    MaskSpawnGroup=false
    Mesh Sampling Type=::NewEnumerator5
    Sampled Position Offset Coordinate Space:ENiagaraSkelMeshTransforms=::NewEnumerator2
    Skeletal Mesh=$System.Skeletal Mesh
    Socket Sampling Mode=::NewEnumerator5
   )
   &ParticleState()
   &ScaleSpriteSize(
    Scale Sprite Size Mode=::NewEnumerator0
    Uniform Scale Factor=&MergeHelper_FloatSwitch(
     0=1
     1=1
     2=&FloatFromCurve(FloatCurve=DataInterface() {propertyValues {Curve {keys [{arriveTangent=2.3562, interpMode=RCIM_Cubic, leaveTangent=2.3562, tangentMode=RCTM_User, time=0, value=0}, <at 1 1>]}}})
     3=&FloatFromCurve(FloatCurve=DataInterface() {propertyValues {Curve {keys [{arriveTangent=2.3562, interpMode=RCIM_Cubic, leaveTangent=2.3562, tangentMode=RCTM_User, time=0, value=0}, <at 1 1>]}}})
     Index=&SpawnGroup()
     Num=::NewEnumerator2
    )
   )
   &ScaleColor(
    Scale Alpha=&Multiply_Float(
     A=&FN_FrontEndFloatModifier(Gameplay=1, Front End=0.78)
     B=&Multiply_Float(
      A=&FNTimeOfDayScalar(Night Scale=0.7, Morning Scale=0.9, Evening Scale=0.8)
      B=&MergeHelper_FloatSwitch(
       0=&FloatFromCurve(
        FloatCurve=DataInterface() {propertyValues {Curve {keys [
            <at 0 0 RCIM_Cubic>
            {arriveTangent=0.1666, interpMode=RCIM_Cubic, leaveTangent=0.1666, tangentMode=RCTM_User, time=0.1, value=1}
            {arriveTangent=0.047, interpMode=RCIM_Cubic, leaveTangent=0.047, tangentMode=RCTM_User, time=0.85, value=1}
            <at 1 0 RCIM_Cubic>
           ]}}}
        Scale Curve=0.12
       )
       1=&FloatFromCurve(
        FloatCurve=DataInterface() {propertyValues {Curve {keys [
            <at 0 0 RCIM_Cubic>
            {arriveTangent=0.1666, interpMode=RCIM_Cubic, leaveTangent=0.1666, tangentMode=RCTM_User, time=0.1, value=1}
            {arriveTangent=0.047, interpMode=RCIM_Cubic, leaveTangent=0.047, tangentMode=RCTM_User, time=0.85, value=1}
            <at 1 0 RCIM_Cubic>
           ]}}}
        Scale Curve=0.12
       )
       2=&FloatFromCurve(
        FloatCurve=DataInterface() {propertyValues {Curve {keys [
            {arriveTangent=-1.8326, interpMode=RCIM_Cubic, leaveTangent=-1.8326, tangentMode=RCTM_User, time=0, value=1}
            {arriveTangent=-0.5891, tangentMode=RCTM_User, time=1, value=0}
           ]}}}
        Scale Curve=0.4
       )
       3=&FloatFromCurve(
        FloatCurve=DataInterface() {propertyValues {Curve {keys [
            {arriveTangent=-1.8326, interpMode=RCIM_Cubic, leaveTangent=-1.8326, tangentMode=RCTM_User, time=0, value=1}
            {arriveTangent=-0.5891, tangentMode=RCTM_User, time=1, value=0}
           ]}}}
        Scale Curve=0.4
       )
       Index=&SpawnGroup()
       Num=::NewEnumerator2
      )
     )
    )
    ScaleA=true
    ScaleRGB=true
   ) {module=ScaleColor002}
   &FN_Light_Attributes(
    Color=<rgba 7 4.0842 2.0679 1>
    Create Color Attribute=true
    Create Exponent Attribute=true
    Create Radius Attribute=true
    Enable Particle Light Enable=true
    Exponent=3
    Light Enabled:NiagaraBool=&SetBoolByIntComparison(A=2, B=&SpawnGroup(), Comparison Type=::NewEnumerator0)
    Radius=&FN_FrontEndFloatModifier(Gameplay=40, Front End=5)
    Scale Color=&FloatFromCurve(
     FloatCurve=DataInterface() {propertyValues {Curve {keys [
         <at 0 0 RCIM_Cubic>
         {arriveTangent=0.1666, interpMode=RCIM_Cubic, leaveTangent=0.1666, tangentMode=RCTM_User, time=0.1, value=1}
         {arriveTangent=0.047, interpMode=RCIM_Cubic, leaveTangent=0.047, tangentMode=RCTM_User, time=0.9, value=1}
         <at 1 0 RCIM_Cubic>
        ]}}}
     Scale Curve=&Multiply_Float(A=1.5, B=&FNTimeOfDayScalar(Night Scale=0.7, Morning Scale=0.9, Evening Scale=0.8))
    )
   )
   /SetVariables(
    Particles.FlareColor=&MakeLinearColorFromVectorAndFloat(
     "Vector (RGB)"=&MakeVectorFromLinearColorRGB(Color=<rgba 2 1.2112 0.7396 1>)
     "Float (Alpha)"=&FloatFromCurve(
      FloatCurve=DataInterface() {propertyValues {Curve {keys [
          {arriveTangent=-1.8326, interpMode=RCIM_Cubic, leaveTangent=-1.8326, tangentMode=RCTM_User, time=0, value=1}
          {arriveTangent=-0.5891, tangentMode=RCTM_User, time=1, value=0}
         ]}}}
      Scale Curve=0.4
     )
    )
    Particles.FlareSize:Vector2f=&Multiply_Vector2D_ByFloat(
     Float=&FloatFromCurve(
      FloatCurve=DataInterface() {propertyValues {Curve {keys [{arriveTangent=1.5285, interpMode=RCIM_Cubic, leaveTangent=1.5285, tangentMode=RCTM_User, time=0, value=0}, <at 1 1>]}}}
      Scale Curve=1.2
     )
     Vector2D=$Particles.Initial.SpriteSize
    )
    Particles.FlashCameraOffset=8
   ) {isSetParameters=true}
  ]
  rendererClasses [&NiagaraSpriteRendererProperties, &NiagaraLightRendererProperties]
  renderers [
   {rendererClass=&NiagaraSpriteRendererProperties}
   {rendererClass=&NiagaraLightRendererProperties, rendererIndex=1}
  ]
 }
}
summary {
 systemName=NS_Emote_ModernMix_Trails
 emitters [
  {emitterName=WristAirglow, rendererClasses [&NiagaraSpriteRendererProperties, &NiagaraLightRendererProperties]}
  {emitterName=HandsRibbonMain, rendererClasses [&NiagaraRibbonRendererProperties]}
  {emitterName=BraceletAnimTrail, rendererClasses [&NiagaraRibbonRendererProperties]}
  {emitterName=HandGlow, rendererClasses [&NiagaraSpriteRendererProperties]}
 ]
 userVariables [
  User.bIsFrontEnd {type=NiagaraBool}
  User.bIsFrontEndPreview {type=NiagaraBool}
  User.RibbonIDOffset {type=NiagaraInt32}
  User.Scale {type=NiagaraFloat, defaultValue=1, description=""}
  User.SkeletalMesh {
   type=NiagaraDataInterfaceSkeletalMesh
   defaultValue {
    struct=NiagaraExt_VariableValue_DataInterface
    value {dataInterface=&self:NiagaraDataInterfaceSkeletalMesh_0}
   }
  }
 ]
}
