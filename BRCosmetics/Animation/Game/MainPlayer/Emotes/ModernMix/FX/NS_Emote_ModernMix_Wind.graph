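# Niagara system for the wind effect of the ModernMix player emote, built from one mesh-particle emitter, WindBurst2.
# Short-lived, flattened meshes spawn above the origin, grow and spin around world Z at a curve-driven rate.
# Their faint alpha fades in and out and is scaled by time of day.
# Dynamic material parameters animate over particle life with random per-particle values, presumably driving the wind material's look.
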NiagaraSystem /BRCosmetics/Animation/Game/MainPlayer/Emotes/ModernMix/FX/NS_Emote_ModernMix_Wind
dependencies {
 usedDataInterfaces [&NiagaraDataInterfaceCurve, &NiagaraDataInterfaceArrayColor, &NiagaraDataInterfaceSkeletalMesh]
 usedDynamicInputs []
 usedModules [
  &self:SystemSpawnScript.SystemScriptSource.SystemScriptGraph.NiagaraNodeAssignment_1.SetVariables_#1
  &SystemState
 ]
}
emitters {WindBurst2 {
  emitterSpawnScript [
   /SetVariables(Emitter.RotationSpeed:NiagaraFloat=&FloatFromCurve(FloatCurve=DataInterface() {propertyValues {Curve {keys [<at 0 -0.5>,<at 0.1 0>]}}}, CurveIndex=$Emitter.NormalizedLoopAge, Scale Curve=0)) {isSetParameters=true}
   /SetVariables(
    Emitter.DarkColor:LinearColor=&FNTimeOfDayColor(
     Night Color=<rgba 0.6458 0.5867 0.5718 1>
     Morning Color=<rgba 0.6458 0.5943 0.5651 1>
     Day Color=<rgba 0.6458 0.5892 0.5449 1>
     Evening Color=<rgba 0.6458 0.6102 0.5819 1>
    )
    Emitter.MainColor:LinearColor=&FNTimeOfDayColor(
     Night Color=&SelectLinearColorFromArray(
      Array Sampling Mode=::NewEnumerator0
      Direct Array Index=&RandomRangeInt(Minimum=0, Maximum=1)
      Color Selection Array=DataInterface() {propertyValues {ColorData [<rgba 1 0.947 0.8646 1>]}}
      Random Seed=0
     )
     Morning Color=&SelectLinearColorFromArray(
      Array Sampling Mode=::NewEnumerator0
      Direct Array Index=&RandomRangeInt(Minimum=0, Maximum=1)
      Color Selection Array=DataInterface() {propertyValues {ColorData [<rgba 1 0.8815 0.8698 1>]}}
      Random Seed=0
     )
     Day Color=&SelectLinearColorFromArray(
      Array Sampling Mode=::NewEnumerator0
      Direct Array Index=&RandomRangeInt(Minimum=0, Maximum=1)
      Color Selection Array=DataInterface() {propertyValues {ColorData [<rgba 1 0.9425 0.901 1>, <rgba 1 0.9106 0.8802 1>]}}
      Random Seed=0
     )
     Evening Color=&SelectLinearColorFromArray(
      Array Sampling Mode=::NewEnumerator0
      Direct Array Index=&RandomRangeInt(Minimum=0, Maximum=1)
      Color Selection Array=DataInterface() {propertyValues {ColorData [<rgba 1 0.9332 0.8802 1>]}}
      Random Seed=0
     )
    )
   ) {isSetParameters=true}
  ]
  emitterUpdateScript [
   &EmitterState(Life Cycle Mode=::NewEnumerator1, Loop Behavior=::NewEnumerator1, Loop Duration=1.2, UseLoopDelay=true)
   &SpawnRate(SpawnRate=3.6, Use Spawn Probability=true)
  ]
  particleSpawnScript [
   &InitializeParticle(
    Color=<rgba 1 1 1 3>
    Lifetime=&RandomRangeFloat(Minimum=1, Maximum=1.3)
    Mesh Scale=&Multiply_VectorByFloat(Vector=<3.6 3.6 0.5>, Float=&RandomRangeFloat(Minimum=0.5, Maximum=0.65))
    Mesh Scale Mode=::NewEnumerator4
    Position Offset=<0 0 50>
    Position Offset Coordinate Space=::Local
    UsePositionOffset=true
    Write Material Random=true
   )
   &InitialMeshOrientation(
    Mesh Axis To Orient=<0 0 3>
    Mesh Orientation Mode=::NewEnumerator2
    Orientation Coordinate Space=::World
    Orientation Vector=<0 0 1>
    Rotation=&RandomRangeVector(Minimum=<0.03 0.03 0>, Maximum=<-0.03 -0.03 1>)
    Use Rotation Vector=true
   )
  ]
  particleUpdateScript [
   &ParticleState()
   &ScaleMeshSize(Scale Factor=&VectorFromFloat(Value=&FloatFromCurve(FloatCurve=DataInterface() {propertyValues {Curve {keys [<at 0 0.4137 RCIM_Cubic>,<at 1 0.7306>]}}})))
   &UpdateMeshOrientation(
    Rotation Rate=&FloatFromCurve(
     FloatCurve=DataInterface() {propertyValues {Curve {keys [<at 0.55 0.4376 RCIM_Cubic>, {arriveTangent=2.9123, interpMode=RCIM_Cubic, leaveTangent=2.9123, tangentMode=RCTM_User, time=1, value=1}]}}}
     CurveIndex=$Emitter.NormalizedLoopAge
     Scale Curve=-1.6154
    )
    Rotation Vector=<0 0 1>
   )
   &ScaleColor(
    Scale Alpha=&FloatFromCurve(
     FloatCurve=DataInterface() {propertyValues {Curve {keys [
         <at 0 0 RCIM_Cubic>
         {interpMode=RCIM_Cubic, tangentMode=RCTM_User, time=0.3258, value=1}
         {interpMode=RCIM_Cubic, tangentMode=RCTM_User, time=0.5806, value=1}
         {tangentMode=RCTM_User, time=1, value=0}
        ]}}}
     Scale Curve=&Multiply_Float(A=0.09, B=&FNTimeOfDayScalar(Night Scale=0.7, Morning Scale=0.9, Evening Scale=0.8))
    )
    Scale RGB=<2 2 2>
    ScaleA=true
    ScaleRGB=true
   )
   &DynamicMaterialParameters(
    Index 0 Param 1=&FloatFromCurve(FloatCurve=DataInterface() {propertyValues {Curve {keys [{tangentMode=RCTM_User, time=0.4, value=1}, {tangentMode=RCTM_User, time=0.9, value=-1}]}}})
    Index 0 Param 2=1
    Index 0 Param 3=&RandomRangeFloat(Minimum=-1, Maximum=1)
    Index 0 Param 4=&RandomRangeFloat(Minimum=-1, Maximum=1)
    Index 1 Param 1=&FloatFromCurve(
     FloatCurve=DataInterface() {propertyValues {Curve {keys [
         {arriveTangent=1.7366, interpMode=RCIM_Cubic, leaveTangent=1.7366, tangentMode=RCTM_User, time=0, value=0}
         {arriveTangent=0.4174, interpMode=RCIM_Cubic, leaveTangent=0.4174, tangentMode=RCTM_User, time=1, value=1}
        ]}}}
     Scale Curve=0.5
    )
    Index 1 Param 2=1
    Module.Index0Float0Write=true
    Module.Index0Float1Write=true
    Module.Index0Float2Write=true
    Module.Index0Float3Write=true
    Module.Index1Float0Write=true
    Module.Index1Float1Write=false
    Module.Index1Float2Write=false
    Module.Index1Float3Write=false
    Write Parameter Index 1=::NewEnumerator1
   ) {module=DynamicMaterialParameters001}
  ]
  rendererClasses [&NiagaraMeshRendererProperties]
  renderers [{rendererClass=&NiagaraMeshRendererProperties, rendererIndex=0}]
 }}
summary {
 systemName=NS_Emote_ModernMix_Wind
 emitters [{emitterName=WindBurst2, rendererClasses [&NiagaraMeshRendererProperties]}]
 userVariables [
  User.bIsFrontEnd {type=NiagaraBool}
  User.bIsFrontEndPreview {type=NiagaraBool}
  User.RibbonIDOffset {type=NiagaraInt32}
  User.Scale {type=NiagaraFloat, defaultValue=1, description=""}
 ]
}
